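# Physics asset for the Dark Bomber Summer backpack cosmetic, defining collision bodies and joints for its dangling ring, strap and keychain.
# A box body on the pack bone anchors the ring, the strap and a keychain chain that runs from keychain root to keychain.
# The keychain carries small wing and tail capsules on tight angular limits, while the root, keychain and strap joints swing more widely and the ring swings mostly on one axis.
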PhysicsAsset /BRCosmetics/Accessories/FORT_Backpacks/Backpack_F_MED_Dark_Bomber_Summer/Meshes/F_MED_Dark_Bomber_Summer_Pack_Physics
bodies {
 dyn_keychain [
  {shapeType=Box, extentX=7.5364, extentY=6.1217, extentZ=1.6489, center=<4.5523 -1.1594 -1.6135>, rotation=<rot -31.1145 22.9935 -3.4857>}
  {shapeType=Box, extentX=3.6863, extentY=3.2625, extentZ=4.16, center=<-0.2739 0 1.2843>, rotation=<rot -28.6703 0 0>}
 ]
 dyn_keychain_root [{shapeType=Capsule, length=0.7127, radius=0.8656, center=<1.0583 0 0>, rotation=<rot 90 0 0>}]
 dyn_l_wing [{shapeType=Capsule, length=0.5, radius=0.5, center=<-0.3988 0.4859 0>, rotation=<rot 90 0 47.3915>}]
 dyn_r_wing [{shapeType=Capsule, length=0.5, radius=0.5, center=<-0.6142 0.1999 0>, rotation=<rot 90 0 28.1242>}]
 dyn_ring [{shapeType=Box, extentX=5.72, extentY=8.4542, extentZ=0.8964, center=<2.9139 0 -0.1839>, rotation=<rot -3.2492 0 0>}]
 dyn_strap [{shapeType=Capsule, length=4.1605, radius=1.8459, center=<2.5658 -0.9223 -0.1152>, rotation=<rot 0 -85.8209 -89.5085>}]
 dyn_tail [{shapeType=Capsule, length=1.9569, radius=0.7543, center=<1.5616 0 -0.8652>, rotation=<rot 68.4066 0 0>}]
 pack [{shapeType=Box, extentX=16.8328, extentY=6.1079, extentZ=29.4893, center=<-0.9976 -3.4444 -4.0141>}]
}
constraints [
 {bone1Name=dyn_ring, bone2Name=pack, swing1LimitDegrees=5, swing2LimitDegrees=80, twistLimitDegrees=15}
 {bone1Name=dyn_keychain_root, bone2Name=pack, swing1LimitDegrees=45, swing2LimitDegrees=45, twistLimitDegrees=30}
 {bone1Name=dyn_keychain, bone2Name=dyn_keychain_root, swing1LimitDegrees=45, swing2LimitDegrees=45, twistLimitDegrees=30}
 {bone1Name=dyn_l_wing, bone2Name=dyn_keychain, swing1LimitDegrees=5, swing2LimitDegrees=5, twistLimitDegrees=5}
 {bone1Name=dyn_r_wing, bone2Name=dyn_keychain, swing1LimitDegrees=5, swing2LimitDegrees=5, twistLimitDegrees=5}
 {bone1Name=dyn_tail, bone2Name=dyn_keychain, swing1LimitDegrees=5, swing2LimitDegrees=5, twistLimitDegrees=5}
 {bone1Name=dyn_strap, bone2Name=pack, swing1LimitDegrees=45, swing2LimitDegrees=45, twistLimitDegrees=45}
]
